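; Burst gameplay cue for the Mantis melee block break, adjusting where its Niagara effect appears on the target pawn.
; On burst, it takes the first spawned Niagara component and, if valid, teleports it to the character mesh's Root socket plus two offsets.
; The vertical offset is along world Z, while the "forward" offset follows the mesh's right vector.
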(fn OnBurst(Niagara)
  (params (MyTarget :type Actor%20Object%20Reference) (Parameters :type Gameplay%20Cue%20Parameters%20Structure%20%28by%20ref%29) (ParticleComponents :type Array%20of%20Niagara%20Particle%20System%20Component%20Object%20References) (AudioComponents :type Array%20of%20Audio%20Component%20Object%20References) (BurstCameraShakeInstance :type Camera%20Shake%20Base%20Object%20Reference) (BurstDecalInstance :type Decal%20Actor%20Object%20Reference))
  (bind _output (Utilities|Array|Get(acopy) ParticleComponents))
  (bind _asfort_pawn (Utilities|Casting|CastToFortPawn MyTarget))
  (Utilities|IsValid _output
    (:Is%20Valid
      (Transformation|SetWorldLocation _output (Utilities|Operators|Add (Transformation|GetSocketLocation (Variables|/Script/Engine.Character|GetMesh _asfort_pawn) "Root") (Math|Vector|MakeVector 0.0 0.0 (LocalVariables|GetVFX%20Vertical%20Offset)) (* (Transformation|GetRightVector (Variables|/Script/Engine.Character|GetMesh _asfort_pawn)) (LocalVariables|GetVFX%20Forward%20Offset))) false true)
      (return))))
