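; Burst gameplay cue that plays impact effects for the Mantis melee combo on a Fort pawn target.
; On execute it traces a short vertical line through the "Hit" socket of the current weapon's mesh to find a surface.
; A hit spawns the first burst particle system at the impact, oriented by the pawn mesh, sets its BurstDirection from VFX Direction, and plays a placeholder hammer sound.
; A trace miss only prints a debug failure message.
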(fn OnExecute
  (params (MyTarget :type Actor%20Object%20Reference) (Parameters :type Gameplay%20Cue%20Parameters%20Structure%20%28by%20ref%29))
  (bind _bursteffectdata_burstparticles (Utilities|Struct|BreakFortBurstEffectData (Variables|GetBurstEffectData_BurstParticles)))
  (bind (_output_condition _output_attachment _output_fxtype _output_niagarasystem) (Utilities|Struct|BreakFortGameplayCueParticleInfo (Utilities|Array|Get(acopy) _bursteffectdata_burstparticles)))
  (bind _asfort_pawn (Utilities|Casting|CastToFortPawn MyTarget)
    (:then
      (bind (_outhit _returnvalue) (Collision|LineTraceForObjects :Start (+ (Transformation|GetSocketLocation (Variables|/Script/FortniteGame.FortWeapon|GetWeaponMesh (Weapon|GetCurrentWeapon _asfort_pawn)) "Hit") "0.000000,0,30.000000") :End (- (Transformation|GetSocketLocation (Variables|/Script/FortniteGame.FortWeapon|GetWeaponMesh (Weapon|GetCurrentWeapon _asfort_pawn)) "Hit") "0.000000,0.000000,100.000000") :ObjectTypes (Utilities|Array|MakeArray "ObjectTypeQuery1") :bTraceComplex false :DrawDebugType "None" :bIgnoreSelf true :TraceColor "(R=1.000000,G=0.000000,B=0.000000,A=1.000000)" :TraceHitColor "(R=0.000000,G=1.000000,B=0.000000,A=1.000000)" :DrawTime 3.0))
      (if _returnvalue
        (bind _returnvalue_2 (Effects|Components|ParticleSystem|SpawnFXAtLocation _asfort_pawn _output_niagarasystem (.ImpactPoint (Collision|BreakHitResult _outhit)) (Math|Rotator|MakeRotfromXZ (Transformation|GetRightVector (Variables|/Script/Engine.Character|GetMesh _asfort_pawn)) (0.0 0.0 1.0))))
        (Effects|Components|ParticleSystem|SetVectorParameter _returnvalue_2 "BurstDirection" (Variables|Default|GetVFX%20Direction))
        ;[1 0 0] TODO: Change out this temp audio
        (Audio|PlaySoundatLocation _returnvalue_2 "/CRD_Melee/Audio/Sounds/Hammer/Melee_Hammer_Primary_Damage_World_Cue.Melee_Hammer_Primary_Damage_World_Cue" (+ (.ImpactPoint (Collision|BreakHitResult _outhit)) "0,0,5.000000"))
        (else
          (Development|PrintString 0 "Impact Cast FAILED!!!!!")))
      (return true))
    (:CastFailed
      (return false))))
