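; Burst gameplay cue for the landing of the Mantis heavy melee attack, which positions the landing VFX on the target pawn.
; Both burst events pass their spawned particle components to a shared FX function.
; FX moves the first valid component to the bottom of the pawn's capsule plus a forward offset, then attaches it to the root component, keeping its world transform.
; OnActive and OnExecute only call the parent.
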(fn FX
  (params (Target :type Actor%20Object%20Reference) (Array :type Array%20of%20FXSystem%20Component%20Object%20References))
  (bind (_returnvalue_x _returnvalue_y _returnvalue_z) (Math|Vector|BreakVector (Transformation|GetActorLocation (LocalVariables|GetFortPawn))))
  (bind _asfort_pawn (Utilities|Casting|CastToFortPawn Target))
  (LocalVariables|SetFortPawn _asfort_pawn)
  (bind _output_get (LocalVariables|SetParticleSystem (Utilities|Array|Get(acopy) Array)))
  (Utilities|IsValid _output_get
    (:Is%20Valid
      (Transformation|SetWorldLocation (LocalVariables|GetParticleSystem) (+ (Math|Vector|MakeVector _returnvalue_x _returnvalue_y (- _returnvalue_z (Components|Capsule|GetUnscaledCapsuleHalfHeight (Variables|/Script/Engine.Character|GetCapsuleComponent (LocalVariables|GetFortPawn))))) (* (Transformation|GetActorForwardVector (LocalVariables|GetFortPawn)) (LocalVariables|GetVFX%20Forward%20Offset))) false true)
      (Transformation|AttachComponentToComponent (LocalVariables|GetParticleSystem) (Variables|/Script/Engine.Actor|GetRootComponent (LocalVariables|GetFortPawn)) "" "KeepWorld" "KeepWorld" "KeepWorld" true))))

(fn OnActive
  (params (MyTarget :type Actor%20Object%20Reference) (Parameters :type Gameplay%20Cue%20Parameters%20Structure%20%28by%20ref%29))
  (return (|CallParentFunction MyTarget Parameters)))

(fn OnBurst(EitherVFX-preferthisversion)
  (params (MyTarget :type Actor%20Object%20Reference) (Parameters :type Gameplay%20Cue%20Parameters%20Structure%20%28by%20ref%29) (ParticleComponents :type Array%20of%20FXSystem%20Component%20Object%20References) (AudioComponents :type Array%20of%20Audio%20Component%20Object%20References) (BurstCameraShakeInstance :type Camera%20Shake%20Base%20Object%20Reference) (BurstDecalInstance :type Decal%20Actor%20Object%20Reference))
  (CallFunction|FX MyTarget ParticleComponents)
  (return))

(fn OnBurst(Niagara)
  (params (MyTarget :type Actor%20Object%20Reference) (Parameters :type Gameplay%20Cue%20Parameters%20Structure%20%28by%20ref%29) (ParticleComponents :type Array%20of%20Niagara%20Particle%20System%20Component%20Object%20References) (AudioComponents :type Array%20of%20Audio%20Component%20Object%20References) (BurstCameraShakeInstance :type Camera%20Shake%20Base%20Object%20Reference) (BurstDecalInstance :type Decal%20Actor%20Object%20Reference))
  (CallFunction|FX MyTarget ParticleComponents)
  (return))

(fn OnExecute
  (params (MyTarget :type Actor%20Object%20Reference) (Parameters :type Gameplay%20Cue%20Parameters%20Structure%20%28by%20ref%29))
  (return (|CallParentFunction MyTarget Parameters)))
