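; Gameplay cue for the Mantis melee weapon drag, spawning ground effects where the pawn's current weapon drags.
; A line trace runs vertically through the weapon mesh's "Hit" socket to find the surface.
; On a hit, it spawns a particle from the cue's burst settings at the impact point, oriented by the character mesh, sets its BurstDirection and attaches a dirt trail to the weapon.
; A miss prints a debug message.
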(fn OnExecute
  (params (MyTarget :type Actor%20Object%20Reference) (Parameters :type Gameplay%20Cue%20Parameters%20Structure%20%28by%20ref%29))
  (bind _bursteffects_burstparticles (Utilities|Struct|BreakGameplayCueNotifyBurstEffects (Variables|GetBurstEffects_BurstParticles)))
  (bind (_output_spawnconditionoverride _output_placementinfooverride _output_niagarasystem) (Utilities|Struct|BreakGameplayCueNotifyParticleInfo (Utilities|Array|Get(aref) _bursteffects_burstparticles)))
  (bind _asfort_pawn (Utilities|Casting|CastToFortPawn MyTarget)
    (:then
      (bind (_outhit _returnvalue) (Collision|LineTraceForObjects :Start (+ (Transformation|GetSocketLocation (Variables|/Script/FortniteGame.FortWeapon|GetWeaponMesh (Weapon|GetCurrentWeapon _asfort_pawn)) "Hit") "0.000000,0,30.000000") :End (- (Transformation|GetSocketLocation (Variables|/Script/FortniteGame.FortWeapon|GetWeaponMesh (Weapon|GetCurrentWeapon _asfort_pawn)) "Hit") "0.000000,0.000000,100.000000") :ObjectTypes (Utilities|Array|MakeArray "ObjectTypeQuery1") :bTraceComplex false :DrawDebugType "None" :bIgnoreSelf true :TraceColor "(R=1.000000,G=0.000000,B=0.000000,A=1.000000)" :TraceHitColor "(R=0.000000,G=1.000000,B=0.000000,A=1.000000)" :DrawTime 3.0))
      (if _returnvalue
        (bind _returnvalue_2 (Effects|Components|ParticleSystem|SpawnFXAtLocation _asfort_pawn _output_niagarasystem (.ImpactPoint (Collision|BreakHitResult _outhit)) (Math|Rotator|MakeRotfromXZ (Transformation|GetRightVector (Variables|/Script/Engine.Character|GetMesh _asfort_pawn)) (0.0 0.0 1.0))))
        (Effects|Components|ParticleSystem|SetVectorParameter _returnvalue_2 "BurstDirection" (Variables|Default|GetVFX%20Direction))
        (Niagara|SpawnSystemAttached "/CRD_Melee/FX/ComboImpacts/NS_Melee_comboDragTrail_Dirt.NS_Melee_comboDragTrail_Dirt" (Variables|/Script/FortniteGame.FortWeapon|GetWeaponMesh (Weapon|GetCurrentWeapon _asfort_pawn)) "None" "0, 0, 0" "0, 0, 0" "SnapToTarget")
        (else
          (Development|PrintString 0 "Impact Cast FAILED!!!!!")))
      (return true))
    (:CastFailed
      (return false))))
