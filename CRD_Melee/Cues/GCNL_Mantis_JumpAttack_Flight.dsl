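; Looping gameplay cue that plays a weapon-specific Niagara effect for the flight phase of the Mantis melee jump attack.
; On burst, Apply clears any previous effect and looks up the current weapon's data tags in a VFX dictionary, reading them directly to avoid appended generic weapon-type tags.
; The effect async-loads, attaches at the weapon's Muzzle if still wanted, and takes custom color from the melee weapon interface.
; Cease-relevant clears that flag and deactivates the effect.
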(event EventBeginPlay)

(event Collision|EventActorBeginOverlap
  (params (OtherActor :type Actor%20Object%20Reference)))

(event EventTick
  (params (DeltaSeconds :type Float%20%28single-precision%29)))

(event Custom|Apply
  (params (Object :type Object%20Reference))
  (bind _vfx_dictionary (Variables|Default|GetVFX%20Dictionary))
  (bind _loopingvfx (Variables|Default|GetLoopingVFX))
  (bind _loopingvfx_2 (Variables|Default|GetLoopingVFX)
    (:then
      (Components|Activation|Deactivate _loopingvfx_2)
      (Variables|Default|SetLoopingVFX 0))
    (:else
      _))
  (Variables|Default|SetGameplay%20Tags)
  (bind _asfort_pawn (Utilities|Casting|CastToFortPawn Object))
  (Variables|Default|SetCurrent%20Weapon (Variables|/Script/FortniteGame.FortPawn|GetCurrentWeapon _asfort_pawn))
  ;[0.074 1 0] NOT using the get weapon data gameplay tags because it will append _other_ tags to the list and return them. In this case, it meant that the "WeaponType.Spear", "WeaponType.Sword", and "WeaponType.Hammer" would append to the list, which would result in whatever one is found first in the map would play. For now, directly access the Gameplay Tags from the Weapon Data.
  (Utilities|IsValid (Variables|/Script/FortniteGame.FortPawn|GetCurrentWeapon _asfort_pawn)
    (:Is%20Valid
      (bind _output_get (Variables|Default|SetGameplay%20Tags (GameplayTags|GetOwnedGameplayTags (Data|GetWeaponData (Variables|/Script/FortniteGame.FortPawn|GetCurrentWeapon _asfort_pawn)) :!class /Script/GameplayTags.GameplayTagAssetInterface)))
      (bind _array_element (Utilities|Array|ForEachLoopwithBreak (GameplayTags|BreakGameplayTagContainer _output_get))
        (:LoopBody
          (if (.ReturnValue (Utilities|Map|Find _vfx_dictionary _array_element))
            (Utilities|FlowControl|Sequence
              (:then_0
                _)
              (:then_1
                _))
            (Variables|Default|SetWantsToBeActive true)
            (bind _object (Utilities|AsyncLoadAsset (Utilities|Map|Find _vfx_dictionary _array_element))
              (:Completed
                (if (Variables|Default|GetWantsToBeActive)
                  (bind _asniagara_system (Utilities|Casting|CastToNiagaraSystem _object))
                  (bind _returnvalue (Niagara|SpawnSystemAttached _asniagara_system (Visual|GetWeaponMesh (Variables|/Script/FortniteGame.FortPawn|GetCurrentWeapon _asfort_pawn)) "Muzzle" "0, 0, 0" "0, 0, 0" "SnapToTarget" true))
                  (Variables|Default|SetLoopingVFX _returnvalue)
                  (bind _ascreative_melee_parent (Utilities|Casting|CastToCreative_Melee_Parent (Variables|Default|GetCurrent%20Weapon)))
                  (bind (_usecustomcolor _customcolor _customvfxintensity) (CreativeMeleeWeaponInterface|GetCustomVFXColor _ascreative_melee_parent :!class /CRD_Melee/Creative_Melee_Weapon_Interface.Creative_Melee_Weapon_Interface_C))
                  (Effects|Components|ParticleSystem|SetColorParameter _loopingvfx "Color" (Math|Color|SelectColor _customcolor "(R=1.000000,G=1.000000,B=1.000000,A=1.000000)" _usecustomcolor))
                  (Effects|Components|ParticleSystem|SetFloatParameter _loopingvfx "UseCustomColor" (select _usecustomcolor 1.0 0.0))
                  (Effects|Components|ParticleSystem|SetFloatParameter _loopingvfx "CustomColorIntensity" (Math|Float|SelectFloat _customvfxintensity 0.0 _usecustomcolor)))))))))))

(event Custom|Remove
  (Variables|Default|SetWantsToBeActive false)
  (bind _loopingvfx (Variables|Default|GetLoopingVFX)
    (:then
      (Components|Activation|Deactivate _loopingvfx))
    (:else)))

(event EventOnRecurring(Niagara)
  (params (MyTarget :type Actor%20Object%20Reference) (Parameters :type Gameplay%20Cue%20Parameters%20Structure%20%28by%20ref%29) (ParticleComponents :type Array%20of%20Niagara%20Particle%20System%20Component%20Object%20References) (AudioComponents :type Array%20of%20Audio%20Component%20Object%20References) (BurstCameraShakeInstance :type Camera%20Shake%20Base%20Object%20Reference) (BurstDecalInstance :type Decal%20Actor%20Object%20Reference)))

(event EventOnRecurring(EitherVFX-preferthisone)
  (params (MyTarget :type Actor%20Object%20Reference) (Parameters :type Gameplay%20Cue%20Parameters%20Structure%20%28by%20ref%29) (ParticleComponents :type Array%20of%20FXSystem%20Component%20Object%20References) (AudioComponents :type Array%20of%20Audio%20Component%20Object%20References) (BurstCameraShakeInstance :type Camera%20Shake%20Base%20Object%20Reference) (BurstDecalInstance :type Decal%20Actor%20Object%20Reference)))

(fn OnBurst
  (params (MyTarget :type Actor%20Object%20Reference) (Parameters :type Gameplay%20Cue%20Parameters%20Structure%20%28by%20ref%29))
  (CallFunction|Apply MyTarget)
  (return true))

(fn OnApplication(EitherVFX-preferthisone)
  (params (MyTarget :type Actor%20Object%20Reference) (Parameters :type Gameplay%20Cue%20Parameters%20Structure%20%28by%20ref%29) (ParticleComponents :type Array%20of%20FXSystem%20Component%20Object%20References) (AudioComponents :type Array%20of%20Audio%20Component%20Object%20References) (BurstCameraShakeInstance :type Camera%20Shake%20Base%20Object%20Reference) (BurstDecalInstance :type Decal%20Actor%20Object%20Reference)))

(fn OnExecute
  (params (MyTarget :type Actor%20Object%20Reference) (Parameters_NormalizedMagnitude :type Float%20%28single-precision%29) (Parameters_RawMagnitude :type Float%20%28single-precision%29) (Parameters_EffectContext :type Gameplay%20Effect%20Context%20Handle%20Structure) (Parameters_MatchedTagName :type Gameplay%20Tag%20Structure) (Parameters_OriginalTag :type Gameplay%20Tag%20Structure) (Parameters_AggregatedSourceTags :type Gameplay%20Tag%20Container%20Structure) (Parameters_AggregatedTargetTags :type Gameplay%20Tag%20Container%20Structure) (Parameters_Location :type Vector :default "0, 0, 0") (Parameters_Normal :type Vector :default "0, 0, 0") (Parameters_Instigator :type Actor%20Object%20Reference) (Parameters_EffectCauser :type Actor%20Object%20Reference) (Parameters_SourceObject :type Object%20Reference) (Parameters_PhysicalMaterial :type Physical%20Material%20Object%20Reference) (Parameters_GameplayEffectLevel :type Integer) (Parameters_AbilityLevel :type Integer) (Parameters_TargetAttachComponent :type Scene%20Component%20Object%20Reference) (Parameters_bReplicateLocationWhenUsingMinimalRepProxy :type Boolean))
  (return false))

(fn OnCeaseRelevant
  (params (MyTarget :type Actor%20Object%20Reference) (Parameters :type Gameplay%20Cue%20Parameters%20Structure%20%28by%20ref%29))
  (CallFunction|Remove)
  (return true))

(fn ConstructionScript)
