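; Gameplay cue for the Mantis melee jump attack that predicts the landing, playing impact effects when the pawn lands.
; On execute it binds landed and movement-mode delegates, unbinding if the previous mode was surface swimming or skydiving.
; Landing spawns particles at floor height ahead of the pawn, skipped if a clambering tag was added.
; The Creative melee weapon supplies radius, block-break and custom color to the particles and an optional floor-traced decal.
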(event Custom|SetImpactRadiusAndBlockBreak
  (params (Comp :type FXSystem%20Component%20Object%20Reference))
  (bind _ascreative_melee_parent (Utilities|Casting|CastToCreative_Melee_Parent (Variables|/Script/FortniteGame.FortPawn|GetCurrentWeapon (Variables|Default|GetFortPawn))))
  (bind _impactradius (CreativeMeleeWeaponInterface|GetCannonballImpactRadius _ascreative_melee_parent :!class /CRD_Melee/Weapons/Creative_Melee_Parent.Creative_Melee_Parent_C))
  (Effects|Components|ParticleSystem|SetFloatParameter Comp "Radius" _impactradius)
  (bind _enabled (CreativeMeleeWeaponInterface|GetSecondaryBreakBlockEnabled _ascreative_melee_parent :!class /CRD_Melee/Weapons/Creative_Melee_Parent.Creative_Melee_Parent_C))
  (Effects|Components|ParticleSystem|SetFloatParameter Comp "Blockbreak" (Math|Conversions|ToFloat(Boolean) _enabled)))

(event EventBeginPlay
  (Variables|Default|SetSpawnedParticleSystem 0)
  (Variables|Default|SetFortPawn 0))

(event Custom|UndbindMovementModeChanged
  (Character|UnbindEventfromMovementModeChangedDelegate (Variables|Default|GetFortPawn) (delegate Custom|MovementModeChangedDelegate_Event_0)))

(event Custom|MovementModeChangedDelegate_Event_0
  (params (Character :type Character%20Object%20Reference) (PrevMovementMode :type EMovementMode%20Enum :default "MOVE_None") (PreviousCustomMode :type Byte))
  (switch Utilities|FlowControl|Switch|SwitchonEMovementMode PrevMovementMode
    (:MOVE_Custom
      (if (or (Utilities|Enum|Equal(Enum) (Utilities|Enum|BytetoEnumEFortCustomMovement PreviousCustomMode) (Variables|Default|GetSurfaceSwimmingMovementMode)) (Utilities|Enum|Equal(Enum) (Utilities|Enum|BytetoEnumEFortCustomMovement PreviousCustomMode) (Variables|Default|GetSkydivingMovementMode)))
        (CallFunction|UndbindMovementModeChanged)
        (CallFunction|UnbindLanded)))))

(event Custom|ExecutionStart
  (params (FortPawn :type Fort%20Pawn%20Object%20Reference))
  (bind _output_get (Variables|Default|SetFortPawn FortPawn))
  (CallFunction|Clambering _output_get)
  (Variables|Default|SetClambering%20Start false)
  (EventDispatchers|AssignOnPawnLanded _output_get (delegate Custom|OnPawnLanded))
  (Character|AssignMovementModeChangedDelegate (Variables|Default|GetFortPawn) (delegate Custom|MovementModeChangedDelegate_Event_0)))

(event Custom|UnbindLanded
  (EventDispatchers|UnbindEventfromOnPawnLanded (Variables|Default|GetFortPawn) (delegate Custom|OnPawnLanded)))

(event Custom|OnPawnLanded
  (params (Hit :type Hit%20Result%20Structure%20%28by%20ref%29))
  (bind (_returnvalue_x _returnvalue_y _returnvalue_z) (Math|Vector|BreakVector (Transformation|GetWorldLocation (Variables|/Script/Engine.Character|GetCapsuleComponent (Variables|Default|GetFortPawn)))))
  (bind (_returnvalue_x_2 _returnvalue_y_2) (Math|Vector|BreakVector (+ (Transformation|GetActorLocation (Variables|Default|GetFortPawn)) (* (Math|Rotator|GetRotationXVector (Transformation|GetActorRotation (Variables|Default|GetFortPawn))) 1.5e+02))))
  (bind _spawnedparticlesystem (Variables|Default|GetSpawnedParticleSystem))
  (CallFunction|UnbindLanded)
  (if (not (Variables|Default|GetClambering%20Start))
    (bind _returnvalue (Effects|Components|ParticleSystem|SpawnFXAtLocation (Variables|Default|GetFortPawn) (Variables|Default|GetParticleSystem) (Math|Vector|MakeVector _returnvalue_x_2 _returnvalue_y_2 (- _returnvalue_z (Components|Capsule|GetUnscaledCapsuleHalfHeight (Variables|/Script/Engine.Character|GetCapsuleComponent (Variables|Default|GetFortPawn))))) (Transformation|GetActorRotation (Variables|Default|GetFortPawn)) "1.000000,1.000000,1.000000" true "None" true))
    (bind _output_get (Variables|Default|SetSpawnedParticleSystem _returnvalue))
    (bind _ascreative_melee_parent (Utilities|Casting|CastToCreative_Melee_Parent (Weapon|GetCurrentWeapon (Variables|Default|GetFortPawn))))
    (bind (_usecustomcolor _customcolor _customvfxintensity) (CreativeMeleeWeaponInterface|GetCustomVFXColor _ascreative_melee_parent :!class /CRD_Melee/Weapons/Creative_Melee_Parent.Creative_Melee_Parent_C))
    (if _usecustomcolor
      (Utilities|IsValid _spawnedparticlesystem
        (:Is%20Valid
          (CallFunction|SpawnImpactDecal (Transformation|GetWorldLocation _returnvalue) true _customcolor)
          (Effects|Components|ParticleSystem|SetColorParameter _spawnedparticlesystem "Color" _customcolor)
          (Effects|Components|ParticleSystem|SetFloatParameter _spawnedparticlesystem "UseCustomColor" 1.0)
          (Effects|Components|ParticleSystem|SetFloatParameter _spawnedparticlesystem "CustomColorIntensity" _customvfxintensity)))
      (else
        (CallFunction|SpawnImpactDecal (Transformation|GetWorldLocation _returnvalue) false "(R=0.000000,G=0.000000,B=0.000000,A=1.000000)")))
    (CallFunction|SetImpactRadiusandBlockBreak _output_get)
    (else
      _))
  (Ability|Async|EndAction (Variables|Default|GetClamberingAction)))

(event Custom|Clambering
  (params (TargetActor :type Actor%20Object%20Reference))
  (bind _asynctaskproxy (Ability|Tasks|WaitGameplayTagAddToActor TargetActor "(TagName=\"Clambering.State.LedgeMoving\")")
    (:then
      (Variables|Default|SetClamberingAction _asynctaskproxy))
    (:Added
      (Variables|Default|SetClambering%20Start true))))

(fn OnBurst
  (params (MyTarget :type Actor%20Object%20Reference) (Parameters :type Gameplay%20Cue%20Parameters%20Structure%20%28by%20ref%29))
  (return false))

(fn OnExecute
  (params (MyTarget :type Actor%20Object%20Reference) (Parameters :type Gameplay%20Cue%20Parameters%20Structure%20%28by%20ref%29))
  (bind _asfort_pawn (Utilities|Casting|CastToFortPawn MyTarget)
    (:then
      (CallFunction|ExecutionStart _asfort_pawn)
      (return true))
    (:CastFailed
      (return false))))

(fn OnCeaseRelevant
  (params (MyTarget :type Actor%20Object%20Reference) (Parameters :type Gameplay%20Cue%20Parameters%20Structure%20%28by%20ref%29))
  (CallFunction|UnbindLanded)
  (return true))

(fn SpawnImpactDecal
  (params (Location :type Vector) (CustomColor :type Boolean) (Color :type Linear%20Color%20Structure))
  (bind _decal_material (LocalVariables|GetDecal%20Material))
  (if (Variables|Default|GetSpawnDecal)
    (bind (_outhit _returnvalue) (Collision|LineTraceForObjects :Start (+ Location "0,0,50.000000") :End (- Location "0,0,100.000000") :ObjectTypes (Utilities|Array|MakeArray "ObjectTypeQuery1") :bTraceComplex false :DrawDebugType "None" :bIgnoreSelf true :TraceColor "(R=1.000000,G=0.000000,B=0.000000,A=1.000000)" :TraceHitColor "(R=0.000000,G=1.000000,B=0.000000,A=1.000000)" :DrawTime 3.0))
    (if _returnvalue
      (bind _ascreative_melee_parent (Utilities|Casting|CastToCreative_Melee_Parent (Variables|/Script/FortniteGame.FortPawn|GetCurrentWeapon (Variables|Default|GetFortPawn))))
      (bind _impactradius (CreativeMeleeWeaponInterface|GetCannonballImpactRadius _ascreative_melee_parent :!class /CRD_Melee/Weapons/Creative_Melee_Parent.Creative_Melee_Parent_C))
      (if (Variables|GetCustomColor)
        (bind _returnvalue_2 (Rendering|Material|CreateDynamicMaterialInstance 0 _decal_material "None" "None" :!class /Script/Engine.KismetMaterialLibrary))
        (Rendering|Material|SetVectorParameterValue _returnvalue_2 "CustomColor" (Variables|GetColor) :!class /Script/Engine.MaterialInstanceDynamic)
        (Rendering|Material|SetScalarParameterValue _returnvalue_2 "UseCustomColor" 1.0 :!class /Script/Engine.MaterialInstanceDynamic)
        (bind _returnvalue_3 (Rendering|Decal|SpawnDecalatLocation 0 _returnvalue_2 (Math|Vector|MakeVector 1e+02 _impactradius _impactradius) (.ImpactPoint (Collision|BreakHitResult _outhit)) (Math|Conversions|RotationFromXVector (* (.Normal (Collision|BreakHitResult _outhit)) -1.0)) 5.0))
        (else
          (bind _returnvalue_4 (Rendering|Decal|SpawnDecalatLocation 0 _decal_material (Math|Vector|MakeVector 1e+02 _impactradius _impactradius) (.ImpactPoint (Collision|BreakHitResult _outhit)) (Math|Conversions|RotationFromXVector (* (.Normal (Collision|BreakHitResult _outhit)) -1.0)) 5.0))))
      (Rendering|Components|Decal|SetFadeOut _returnvalue_4 0.2 4.5 true)
      (return)
      (else
        (Development|PrintString 0 "Jump attack hit but floor too far away")))))

(fn ConstructionScript)
