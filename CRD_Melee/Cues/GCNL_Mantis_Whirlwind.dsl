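; Looping gameplay cue that plays whirlwind Niagara effects for Creative melee weapons.
; On loop start, it matches gameplay tags read directly from weapon data against scale maps, avoiding extra appended tags.
; It attaches a radius-scaled whirlwind to the pawn and a width-scaled trail to the weapon, applying interface-driven custom color to both and a block-break flag to the whirlwind.
; On removal, it deactivates both effects.
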(event EventBeginPlay)

(event Custom|SetFXBlockBreakSettings
  (bind _enabled (CreativeMeleeWeaponInterface|GetSecondaryBreakBlockEnabled (Variables|Default|GetCurrentMeleeWeapon) :!class /CRD_Melee/Weapons/Creative_Melee_Parent.Creative_Melee_Parent_C))
  (Effects|Components|ParticleSystem|SetFloatParameter (Variables|Default|GetLoopingVFX1) "Blockbreak" (Math|Conversions|ToFloat(Boolean) _enabled)))

(event Collision|EventActorBeginOverlap
  (params (OtherActor :type Actor%20Object%20Reference)))

(event EventTick
  (params (DeltaSeconds :type Float%20%28single-precision%29)))

(event EventOnLoopingStart(EitherVFX-preferthisversion)
  (params (MyTarget :type Actor%20Object%20Reference) (Parameters :type Gameplay%20Cue%20Parameters%20Structure%20%28by%20ref%29) (ParticleComponents :type Array%20of%20FXSystem%20Component%20Object%20References) (AudioComponents :type Array%20of%20Audio%20Component%20Object%20References))
  (bind _vfx_scale (Variables|Default|GetVFX%20Scale))
  (Variables|Default|SetGameplay%20Tags)
  (bind _asfort_pawn (Utilities|Casting|CastToFortPawn MyTarget))
  ;[0.074 1 0] NOT using the get weapon data gameplay tags because it will append _other_ tags to the list and return them. In this case, it meant that the "WeaponType.Spear", "WeaponType.Sword", and "WeaponType.Hammer" would append to the list, which would result in whatever one is found first in the map would play. For now, directly access the Gameplay Tags from the Weapon Data.
  (bind _ascreative_melee_parent (Utilities|Casting|CastToCreative_Melee_Parent (Variables|/Script/FortniteGame.FortPawn|GetCurrentWeapon _asfort_pawn)))
  (Variables|Default|SetCurrentMeleeWeapon _ascreative_melee_parent)
  (bind _output_get (Variables|Default|SetGameplay%20Tags (GameplayTags|GetOwnedGameplayTags (Data|GetWeaponData (Variables|/Script/FortniteGame.FortPawn|GetCurrentWeapon _asfort_pawn)) :!class /Script/GameplayTags.GameplayTagAssetInterface)))
  (bind _array_element (Utilities|Array|ForEachLoopwithBreak (GameplayTags|BreakGameplayTagContainer _output_get))
    (:LoopBody
      (if (.ReturnValue (Utilities|Map|Find _vfx_scale _array_element))
        (Utilities|FlowControl|Sequence
          (:then_0
            _)
          (:then_1
            _))
        (bind _returnvalue (Niagara|SpawnSystemAttached (Variables|Default|GetWhirlwind%20VFX) (Variables|/Script/Engine.Character|GetMesh _asfort_pawn) "Root" "0, 0, 0" "0, 0, 0" "SnapToTarget" true))
        (bind _output_get_2 (Variables|Default|SetLoopingVFX1 _returnvalue))
        (CallFunction|CheckandSetCustomVFXColor (Variables|/Script/FortniteGame.FortPawn|GetCurrentWeapon _asfort_pawn) _output_get_2)
        (Effects|Components|ParticleSystem|SetFloatParameter _output_get_2 "RadiusScale" (Utilities|Map|Find _vfx_scale _array_element))
        (bind _returnvalue_2 (Niagara|SpawnSystemAttached (Variables|Default|GetWhirlwind%20Trail%20VFX) (Visual|GetWeaponMesh (Variables|/Script/FortniteGame.FortPawn|GetCurrentWeapon _asfort_pawn)) "FX_TrailEnd" "0, 0, 0" "0, 0, 0" "SnapToTarget" true))
        (bind _output_get_3 (Variables|Default|SetLoopingVFX2 _returnvalue_2))
        (CallFunction|CheckandSetCustomVFXColor (Variables|/Script/FortniteGame.FortPawn|GetCurrentWeapon _asfort_pawn) _output_get_3)
        (Effects|Components|ParticleSystem|SetFloatParameter _output_get_3 "TrailWidth" (Utilities|Map|Find (Variables|Default|GetVFX%20Trail%20Scale) _array_element))
        (CallFunction|SetFXBlockBreakSettings)))))

(event EventOnRemoval(EitherVFX-preferthisversion)
  (params (MyTarget :type Actor%20Object%20Reference) (Parameters :type Gameplay%20Cue%20Parameters%20Structure%20%28by%20ref%29) (ParticleComponents :type Array%20of%20FXSystem%20Component%20Object%20References) (AudioComponents :type Array%20of%20Audio%20Component%20Object%20References) (BurstCameraShakeInstance :type Camera%20Shake%20Base%20Object%20Reference) (BurstDecalInstance :type Decal%20Actor%20Object%20Reference))
  (bind _loopingvfx2 (Variables|Default|GetLoopingVFX2))
  (bind _loopingvfx1 (Variables|Default|GetLoopingVFX1))
  (Utilities|FlowControl|Sequence
    (:then_0
      (Utilities|IsValid _loopingvfx1
        (:Is%20Valid
          (Components|Activation|Deactivate _loopingvfx1))))
    (:then_1
      (Utilities|IsValid _loopingvfx2
        (:Is%20Valid
          (Components|Activation|Deactivate _loopingvfx2))))))

(fn CheckAndSetCustomVFXColor
  (params (Weapon :type Fort%20Weapon%20Object%20Reference) (FX%20Component :type FXSystem%20Component%20Object%20Reference))
  (bind _ascreative_melee_parent (Utilities|Casting|CastToCreative_Melee_Parent Weapon))
  (bind (_usecustomcolor _customcolor _customvfxintensity) (CreativeMeleeWeaponInterface|GetCustomVFXColor _ascreative_melee_parent :!class /CRD_Melee/Creative_Melee_Weapon_Interface.Creative_Melee_Weapon_Interface_C))
  (Effects|Components|ParticleSystem|SetColorParameter FX%20Component "Color" (Math|Color|SelectColor _customcolor "(R=1.000000,G=1.000000,B=1.000000,A=1.000000)" _usecustomcolor))
  (Effects|Components|ParticleSystem|SetFloatParameter FX%20Component "UseCustomColor" (select _usecustomcolor 1.0 0.0))
  (Effects|Components|ParticleSystem|SetFloatParameter FX%20Component "CustomColorIntensity" (Math|Float|SelectFloat _customvfxintensity 0.0 _usecustomcolor)))

(fn OnApplication(EitherVFX-preferthisone)
  (params (MyTarget :type Actor%20Object%20Reference) (Parameters :type Gameplay%20Cue%20Parameters%20Structure%20%28by%20ref%29) (ParticleComponents :type Array%20of%20FXSystem%20Component%20Object%20References) (AudioComponents :type Array%20of%20Audio%20Component%20Object%20References) (BurstCameraShakeInstance :type Camera%20Shake%20Base%20Object%20Reference) (BurstDecalInstance :type Decal%20Actor%20Object%20Reference)))

(fn ConstructionScript)
